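; Blueprint interface in the Landmass landscape Blueprint brush tools, named as an editor interface for brush actors.
; It declares two functions, MoveUp and MoveDown, each taking a Brush actor reference and returning nothing.
; It holds no logic, and the names suggest implementers shift the given brush up or down, likely in an ordering of brushes.
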(fn MoveDown
  (params (Brush :type Actor%20Object%20Reference)))

(fn MoveUp
  (params (Brush :type Actor%20Object%20Reference)))
