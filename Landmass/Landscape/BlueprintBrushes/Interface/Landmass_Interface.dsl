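; Blueprint interface for Landmass landscape brushes, declaring the functions that implementers expose for shape masks, erosion and updates.
; Declares getting a shape mask and setting one from a texture, points and extents, plus copying edge mask parameters to a dynamic material instance.
; Also declares render dependency queries, erosion layer rendering given brush parameters, extents and a Landmass Brush Manager, and a brush update call.
; The functions hold no logic here.
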(fn CopyEdgeMaskParameters
  (params (MID :type Material%20Instance%20Dynamic%20Object%20Reference)))

(fn GetShapeMask
  (return ""))

(fn GetRenderDependencies
  (return 0))

(fn RenderErosionLayer
  (params (In%20Parameters :type Landscape%20Brush%20Parameters%20Structure) (Extents :type Vector%204%20Structure) (BrushManager :type Landmass%20Brush%20Manager%20Object%20Reference))
  (return false))

(fn SetShapeMask
  (params (Texture :type Texture%202D%20Object%20Reference) (Points :type Array%20of%20Vectors) (NewExtents :type Vector%204%20Structure)))

(fn UpdateBrushes)
